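# Level Sequence for a portrait emote camera shot in the Pilgrim camera manager, framing a Fortnite player mannequin.
# A spawned Cine Camera, attached via an attach track and driven by a camera cut, pushes in and settles back slightly while easing its tilted roll.
# Focal length widens during the move, focus distance is keyed manually, and a camera shake track layers on top.
# The mannequin has attach and animation tracks, alongside a static spotlight relocator actor.
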LevelSequence /PilgrimCore/CameraManager/Sequences/EmoteShots/LS_PilgrimCam_EM_01_Portrait
bindings [
 spotLight_relocator {
  id=#1
  possessed_class=Actor
  tracks [
   {type=MovieSceneSpawnTrack, sections [{type=MovieSceneSpawnSection, channels [None {type=MovieSceneScriptingBoolChannel, default=true}]}]}
   {
    type=MovieScene3DTransformTrack
    sections [{
      type=MovieScene3DTransformSection
      channels [
       Location.X {type=MovieSceneScriptingDoubleChannel, default=0}
       Location.Y {type=MovieSceneScriptingDoubleChannel, default=0}
       Location.Z {type=MovieSceneScriptingDoubleChannel, default=0}
       Rotation.X {type=MovieSceneScriptingDoubleChannel, default=0}
       Rotation.Y {type=MovieSceneScriptingDoubleChannel, default=0}
       Rotation.Z {type=MovieSceneScriptingDoubleChannel, default=0}
       Scale.X {type=MovieSceneScriptingDoubleChannel, default=1}
       Scale.Y {type=MovieSceneScriptingDoubleChannel, default=1}
       Scale.Z {type=MovieSceneScriptingDoubleChannel, default=1}
      ]
     }]
   }
  ]
 }
 CameraComponent {
  id=#2
  parent_id=#3
  possessed_class=CineCameraComponent
  tracks [
   {
    type=MovieSceneFloatTrack
    display_name=Current Focal Length
    sections [{
      type=MovieSceneFloatSection
      channels [None {type=MovieSceneScriptingFloatChannel, default=28.7463, keys [[0,35.6684],[24,35.5048],[52,28.7463]]}]
     }]
   }
   {
    type=MovieSceneFloatTrack
    display_name="Manual Focus Distance (Focus Settings)"
    sections [{type=MovieSceneFloatSection, channels [None {type=MovieSceneScriptingFloatChannel, default=330.1125, keys [[0,228.6909],[33,330.1125]]}]}]
   }
   {
    type=MovieSceneFloatTrack
    display_name=Current Aperture
    sections [{type=MovieSceneFloatSection, channels [None {type=MovieSceneScriptingFloatChannel, default=1}]}]
   }
   {type=MovieSceneCameraShakeTrack, sections [{type=MovieSceneCameraShakeSection, end_frame=8051}]}
  ]
 }
 Cine Camera Actor {
  id=#3
  possessed_class=CineCameraActor
  tracks [
   {type=MovieSceneSpawnTrack, sections [{type=MovieSceneSpawnSection, channels [None {type=MovieSceneScriptingBoolChannel, default=true}]}]}
   {type=MovieScene3DAttachTrack, sections [{type=MovieScene3DAttachSection, end_frame=7874, start_frame=0}]}
   {
    type=MovieScene3DTransformTrack
    sections [{
      type=MovieScene3DTransformSection
      channels [
       Location.X {type=MovieSceneScriptingDoubleChannel, default=415.0838, keys [[0,491.9449],[25,329.8914],[33,338.5386]]}
       Location.Y {type=MovieSceneScriptingDoubleChannel, default=56.3143, keys [[0,75.5237],[25.285,37.1839],[33,37.1839]]}
       Location.Z {type=MovieSceneScriptingDoubleChannel, default=0, keys [[0,1.8783],[25.285,-1.8706],[33,-1.8706]]}
       Rotation.X {type=MovieSceneScriptingDoubleChannel, default=19.3208, keys [[0,19.3208],[33,14.7615]]}
       Rotation.Y {type=MovieSceneScriptingDoubleChannel, default=-1.3581, const [-1.3581,3]}
       Rotation.Z {type=MovieSceneScriptingDoubleChannel, default=-175, keys [[0,-170],[25,-172],[33,-175]]}
       Scale.X {type=MovieSceneScriptingDoubleChannel, default=1, const [1,3]}
       Scale.Y {type=MovieSceneScriptingDoubleChannel, default=1, const [1,3]}
       Scale.Z {type=MovieSceneScriptingDoubleChannel, default=1, const [1,3]}
      ]
     }]
   }
  ]
 }
 Fort Player Mannequin {
  id=#4
  possessed_class=FortPlayerMannequin
  tracks [
   {
    type=MovieScene3DTransformTrack
    sections [{
      type=MovieScene3DTransformSection
      channels [
       Location.X {type=MovieSceneScriptingDoubleChannel, default=0}
       Location.Y {type=MovieSceneScriptingDoubleChannel, default=0}
       Location.Z {type=MovieSceneScriptingDoubleChannel, default=-90.7379}
       Rotation.X {type=MovieSceneScriptingDoubleChannel, default=0}
       Rotation.Y {type=MovieSceneScriptingDoubleChannel, default=0}
       Rotation.Z {type=MovieSceneScriptingDoubleChannel, default=-90}
       Scale.X {type=MovieSceneScriptingDoubleChannel, default=1}
       Scale.Y {type=MovieSceneScriptingDoubleChannel, default=1}
       Scale.Z {type=MovieSceneScriptingDoubleChannel, default=1}
      ]
     }]
   }
   {type=MovieSceneSkeletalAnimationTrack, sections [{type=MovieSceneSkeletalAnimationSection, end_frame=7874, start_frame=0}]}
   {type=MovieSceneSpawnTrack, sections [{type=MovieSceneSpawnSection, channels [None {type=MovieSceneScriptingBoolChannel, default=true}]}]}
   {type=MovieScene3DAttachTrack, sections [{type=MovieScene3DAttachSection, end_frame=7873, start_frame=0}]}
  ]
 }
]
display_rate {denominator=1, numerator=60}
tick_resolution {denominator=1, numerator=24000}
tracks [{type=MovieSceneCameraCutTrack, sections [{type=MovieSceneCameraCutSection, end_frame=8420, start_frame=0}]}]
