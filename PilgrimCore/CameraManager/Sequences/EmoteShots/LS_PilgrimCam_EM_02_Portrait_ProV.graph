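# Pilgrim camera Level Sequence for an emote portrait shot (EM_02, ProV variant), framing a character through a single camera cut.
# A spawned CineCamera actor holds a fixed, attached transform while its camera component carries the motion, pushing in and returning to its starting position.
# Focal length and manual focus distance are keyed along the move, with a constant wide aperture and a camera shake layered on.
# An attached, animated FortPlayerMannequin and a spotlight relocator actor accompany the camera.
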LevelSequence /PilgrimCore/CameraManager/Sequences/EmoteShots/LS_PilgrimCam_EM_02_Portrait_ProV
bindings [
 spotLight_relocator {
  id=#1
  possessed_class=Actor
  tracks [
   {type=MovieSceneSpawnTrack, sections [{type=MovieSceneSpawnSection, channels [None {type=MovieSceneScriptingBoolChannel, default=true}]}]}
   {
    type=MovieScene3DTransformTrack
    sections [{
      type=MovieScene3DTransformSection
      channels [
       Location.X {type=MovieSceneScriptingDoubleChannel, default=0}
       Location.Y {type=MovieSceneScriptingDoubleChannel, default=0}
       Location.Z {type=MovieSceneScriptingDoubleChannel, default=0}
       Rotation.X {type=MovieSceneScriptingDoubleChannel, default=0}
       Rotation.Y {type=MovieSceneScriptingDoubleChannel, default=0}
       Rotation.Z {type=MovieSceneScriptingDoubleChannel, default=0}
       Scale.X {type=MovieSceneScriptingDoubleChannel, default=1}
       Scale.Y {type=MovieSceneScriptingDoubleChannel, default=1}
       Scale.Z {type=MovieSceneScriptingDoubleChannel, default=1}
      ]
     }]
   }
  ]
 }
 CameraComponent {
  id=#2
  parent_id=#3
  possessed_class=CineCameraComponent
  tracks [
   {
    type=MovieSceneFloatTrack
    display_name=Current Focal Length
    sections [{
      type=MovieSceneFloatSection
      channels [None {type=MovieSceneScriptingFloatChannel, default=34.944, keys [[0,34.944],[445,23.5703],[823,34.944],[823.3575,23.5703]]}]
     }]
   }
   {
    type=MovieSceneFloatTrack
    display_name="Manual Focus Distance (Focus Settings)"
    sections [{
      type=MovieSceneFloatSection
      channels [None {type=MovieSceneScriptingFloatChannel, default=231.8993, keys [[0,473.4482],[444.9425,231.8993],[823,473.4482],[823.3575,441.0363]]}]
     }]
   }
   {
    type=MovieSceneFloatTrack
    display_name=Current Aperture
    sections [{type=MovieSceneFloatSection, channels [None {type=MovieSceneScriptingFloatChannel, default=0.8}]}]
   }
   {
    type=MovieScene3DTransformTrack
    sections [{
      type=MovieScene3DTransformSection
      channels [
       Location.X {type=MovieSceneScriptingDoubleChannel, default=-255.3156, thinned_from=4, keys [[0,-256.2995],[444.9425,-43.1848],[823.3575,-256.2995]]}
       Location.Y {type=MovieSceneScriptingDoubleChannel, default=-119.7192, thinned_from=4, keys [[0,-123.8044],[444.9425,-70.3875],[823.3575,-123.8044]]}
       Location.Z {type=MovieSceneScriptingDoubleChannel, default=68.2993, thinned_from=4, keys [[0,-57.2218],[444.9425,-52.4345],[823.3575,-57.2218]]}
       Rotation.X {type=MovieSceneScriptingDoubleChannel, default=20.2876, thinned_from=4, keys [[0,19.6095],[444.9425,19.6669],[823.3575,19.6095]]}
       Rotation.Y {type=MovieSceneScriptingDoubleChannel, default=-15.1788, keys [[0,1.2482],[444.9425,4.4703],[823,1.2482],[823.3575,1.2482]]}
       Rotation.Z {type=MovieSceneScriptingDoubleChannel, default=13.8593, thinned_from=4, keys [[0,14.0712],[444.9425,15.2222],[823.3575,14.0712]]}
       Scale.X {type=MovieSceneScriptingDoubleChannel, default=1, const [1,4]}
       Scale.Y {type=MovieSceneScriptingDoubleChannel, default=1, const [1,4]}
       Scale.Z {type=MovieSceneScriptingDoubleChannel, default=1, const [1,4]}
      ]
     }]
   }
   {type=MovieSceneCameraShakeTrack, sections [{type=MovieSceneCameraShakeSection, end_frame=9796}]}
  ]
 }
 CineCameraActor1 {
  id=#3
  possessed_class=CineCameraActor
  tracks [
   {
    type=MovieScene3DTransformTrack
    sections [{
      type=MovieScene3DTransformSection
      channels [
       Location.X {type=MovieSceneScriptingDoubleChannel, default=-5150.5862, const [182.1679,1]}
       Location.Y {type=MovieSceneScriptingDoubleChannel, default=-552.7704, const [-48.9434,1]}
       Location.Z {type=MovieSceneScriptingDoubleChannel, default=93, const [93,1]}
       Rotation.X {type=MovieSceneScriptingDoubleChannel, default=0, const [-19.3525,1]}
       Rotation.Y {type=MovieSceneScriptingDoubleChannel, default=-13.2838, const [-2.6086,1]}
       Rotation.Z {type=MovieSceneScriptingDoubleChannel, default=179.1, const [163.4301,1]}
       Scale.X {type=MovieSceneScriptingDoubleChannel, default=1, const [1,1]}
       Scale.Y {type=MovieSceneScriptingDoubleChannel, default=1, const [1,1]}
       Scale.Z {type=MovieSceneScriptingDoubleChannel, default=1, const [1,1]}
      ]
     }]
   }
   {type=MovieSceneSpawnTrack, sections [{type=MovieSceneSpawnSection, channels [None {type=MovieSceneScriptingBoolChannel, default=true}]}]}
   {type=MovieScene3DAttachTrack, sections [{type=MovieScene3DAttachSection, end_frame=7874, start_frame=0}]}
  ]
 }
 Fort Player Mannequin {
  id=#4
  possessed_class=FortPlayerMannequin
  tracks [
   {
    type=MovieScene3DTransformTrack
    sections [{
      type=MovieScene3DTransformSection
      channels [
       Location.X {type=MovieSceneScriptingDoubleChannel, default=0}
       Location.Y {type=MovieSceneScriptingDoubleChannel, default=0}
       Location.Z {type=MovieSceneScriptingDoubleChannel, default=-90.7379}
       Rotation.X {type=MovieSceneScriptingDoubleChannel, default=0}
       Rotation.Y {type=MovieSceneScriptingDoubleChannel, default=0}
       Rotation.Z {type=MovieSceneScriptingDoubleChannel, default=-90}
       Scale.X {type=MovieSceneScriptingDoubleChannel, default=1}
       Scale.Y {type=MovieSceneScriptingDoubleChannel, default=1}
       Scale.Z {type=MovieSceneScriptingDoubleChannel, default=1}
      ]
     }]
   }
   {type=MovieSceneSkeletalAnimationTrack, sections [{type=MovieSceneSkeletalAnimationSection, end_frame=7874, start_frame=0}]}
   {type=MovieSceneSpawnTrack, sections [{type=MovieSceneSpawnSection, channels [None {type=MovieSceneScriptingBoolChannel, default=true}]}]}
   {type=MovieScene3DAttachTrack, sections [{type=MovieScene3DAttachSection, end_frame=7873, start_frame=0}]}
  ]
 }
]
display_rate {denominator=1, numerator=60}
tick_resolution {denominator=1, numerator=24000}
tracks [{type=MovieSceneCameraCutTrack, sections [{type=MovieSceneCameraCutSection, end_frame=8420, start_frame=0}]}]
