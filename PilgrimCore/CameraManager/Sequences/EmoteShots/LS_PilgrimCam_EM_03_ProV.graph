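# Level Sequence for a Pilgrim emote camera shot (EM_03_ProV), cutting to a spawned CineCamera.
# The camera actor holds a fixed transform with an attach track, while its CineCameraComponent carries the keyed sweeping move, easing focal length and manual focus distance, plus a camera shake.
# A spawned spotLight_relocator actor sits at identity transform.
# An animated, attached Fort player mannequin has spawning off, apparently as a preview stand-in.
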LevelSequence /PilgrimCore/CameraManager/Sequences/EmoteShots/LS_PilgrimCam_EM_03_ProV
bindings [
 spotLight_relocator {
  id=#1
  possessed_class=Actor
  tracks [
   {type=MovieSceneSpawnTrack, sections [{type=MovieSceneSpawnSection, channels [None {type=MovieSceneScriptingBoolChannel, default=true}]}]}
   {
    type=MovieScene3DTransformTrack
    sections [{
      type=MovieScene3DTransformSection
      channels [
       Location.X {type=MovieSceneScriptingDoubleChannel, default=0}
       Location.Y {type=MovieSceneScriptingDoubleChannel, default=0}
       Location.Z {type=MovieSceneScriptingDoubleChannel, default=0}
       Rotation.X {type=MovieSceneScriptingDoubleChannel, default=0}
       Rotation.Y {type=MovieSceneScriptingDoubleChannel, default=0}
       Rotation.Z {type=MovieSceneScriptingDoubleChannel, default=0}
       Scale.X {type=MovieSceneScriptingDoubleChannel, default=1}
       Scale.Y {type=MovieSceneScriptingDoubleChannel, default=1}
       Scale.Z {type=MovieSceneScriptingDoubleChannel, default=1}
      ]
     }]
   }
  ]
 }
 CameraComponent {
  id=#2
  parent_id=#3
  possessed_class=CineCameraComponent
  tracks [
   {
    type=MovieSceneFloatTrack
    display_name=Current Focal Length
    sections [{
      type=MovieSceneFloatSection
      channels [None {type=MovieSceneScriptingFloatChannel, default=11.9666, keys [[0,11.9666],[42,12.8918],[72,12.9749]]}]
     }]
   }
   {
    type=MovieSceneFloatTrack
    display_name="Manual Focus Distance (Focus Settings)"
    sections [{
      type=MovieSceneFloatSection
      channels [None {type=MovieSceneScriptingFloatChannel, default=243.1276, keys [[0,147.0305],[42,235.2059],[72,243.1276]]}]
     }]
   }
   {
    type=MovieSceneFloatTrack
    display_name=Current Aperture
    sections [{type=MovieSceneFloatSection, channels [None {type=MovieSceneScriptingFloatChannel, default=0.8}]}]
   }
   {
    type=MovieScene3DTransformTrack
    sections [{
      type=MovieScene3DTransformSection
      channels [
       Location.X {type=MovieSceneScriptingDoubleChannel, default=-255.3156, keys [[0,128.6375],[11,36.8337],[24,-32.8022],[42,-38.0251],[72,-38.597]]}
       Location.Y {type=MovieSceneScriptingDoubleChannel, default=-119.7192, keys [[0,122.5946],[11,90.0971],[24,-29.1921],[42,-81.9194],[72,-86.0676]]}
       Location.Z {type=MovieSceneScriptingDoubleChannel, default=68.2993, keys [[0,-94.8045],[11,-94.2007],[24,-73.758],[42,-53.0555],[72,-50.2108]]}
       Rotation.X {type=MovieSceneScriptingDoubleChannel, default=20.2876, keys [[0,3.6722],[11,11.8051],[24,17.9822],[42,20.0037],[72,20.2279]]}
       Rotation.Y {type=MovieSceneScriptingDoubleChannel, default=-15.1788, keys [[0,27.399],[11,18.6961],[24,6.6884],[42,3.1583],[72,2.9955]]}
       Rotation.Z {type=MovieSceneScriptingDoubleChannel, default=13.8593, keys [[0,-66.9842],[11,-31.2474],[24,9.3561],[42,21.6203],[72,22.4069]]}
       Scale.X {type=MovieSceneScriptingDoubleChannel, default=1, const [1,5]}
       Scale.Y {type=MovieSceneScriptingDoubleChannel, default=1, const [1,5]}
       Scale.Z {type=MovieSceneScriptingDoubleChannel, default=1, const [1,5]}
      ]
     }]
   }
   {type=MovieSceneCameraShakeTrack, sections [{type=MovieSceneCameraShakeSection, end_frame=8198}]}
  ]
 }
 CineCameraActor1 {
  id=#3
  possessed_class=CineCameraActor
  tracks [
   {
    type=MovieScene3DTransformTrack
    sections [{
      type=MovieScene3DTransformSection
      channels [
       Location.X {type=MovieSceneScriptingDoubleChannel, default=-5150.5862, const [182.1679,1]}
       Location.Y {type=MovieSceneScriptingDoubleChannel, default=-552.7704, const [-48.9434,1]}
       Location.Z {type=MovieSceneScriptingDoubleChannel, default=89, const [89,1]}
       Rotation.X {type=MovieSceneScriptingDoubleChannel, default=0, const [-19.3525,1]}
       Rotation.Y {type=MovieSceneScriptingDoubleChannel, default=-13.2838, const [-2.6086,1]}
       Rotation.Z {type=MovieSceneScriptingDoubleChannel, default=179.1, const [163.4301,1]}
       Scale.X {type=MovieSceneScriptingDoubleChannel, default=1, const [1,1]}
       Scale.Y {type=MovieSceneScriptingDoubleChannel, default=1, const [1,1]}
       Scale.Z {type=MovieSceneScriptingDoubleChannel, default=1, const [1,1]}
      ]
     }]
   }
   {type=MovieSceneSpawnTrack, sections [{type=MovieSceneSpawnSection, channels [None {type=MovieSceneScriptingBoolChannel, default=true}]}]}
   {type=MovieScene3DAttachTrack, sections [{type=MovieScene3DAttachSection, end_frame=7874, start_frame=0}]}
  ]
 }
 Fort Player Mannequin {
  id=#4
  possessed_class=FortPlayerMannequin
  tracks [
   {
    type=MovieScene3DTransformTrack
    sections [{
      type=MovieScene3DTransformSection
      channels [
       Location.X {type=MovieSceneScriptingDoubleChannel, default=0}
       Location.Y {type=MovieSceneScriptingDoubleChannel, default=0}
       Location.Z {type=MovieSceneScriptingDoubleChannel, default=-90.7379}
       Rotation.X {type=MovieSceneScriptingDoubleChannel, default=0}
       Rotation.Y {type=MovieSceneScriptingDoubleChannel, default=0}
       Rotation.Z {type=MovieSceneScriptingDoubleChannel, default=-90}
       Scale.X {type=MovieSceneScriptingDoubleChannel, default=1}
       Scale.Y {type=MovieSceneScriptingDoubleChannel, default=1}
       Scale.Z {type=MovieSceneScriptingDoubleChannel, default=1}
      ]
     }]
   }
   {type=MovieSceneSkeletalAnimationTrack, sections [{type=MovieSceneSkeletalAnimationSection, end_frame=7874, start_frame=0}]}
   {type=MovieSceneSpawnTrack, sections [{type=MovieSceneSpawnSection, channels [None {type=MovieSceneScriptingBoolChannel, default=false}]}]}
   {type=MovieScene3DAttachTrack, sections [{type=MovieScene3DAttachSection, end_frame=7873, start_frame=0}]}
  ]
 }
]
display_rate {denominator=1, numerator=60}
tick_resolution {denominator=1, numerator=24000}
tracks [{type=MovieSceneCameraCutTrack, sections [{type=MovieSceneCameraCutSection, end_frame=8420, start_frame=0}]}]
